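; BPI_VividRazorCore_VaultEvents is a Blueprint interface for querying and triggering vault events, covering both vaults and train vaults.
; IsTrainVault and IsVaultStarted report vault type and state.
; RequestVaultStart and RequestTrainVaultStart each take a vault start payload struct and return a boolean.
; VaultOpened is a parameterless notification with no return value, and the interface holds only signatures, leaving behavior to implementers.
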(fn IsTrainVault
  (return false))

(fn IsVaultStarted
  (return false))

(fn RequestTrainVaultStart
  (params (Payload :type Struct%20Vivid%20Razor%20Core%20Vault%20Start%20Payload%20Structure))
  (return false))

(fn RequestVaultStart
  (params (Payload :type Struct%20Vivid%20Razor%20Core%20Vault%20Start%20Payload%20Structure))
  (return false))

(fn VaultOpened)
